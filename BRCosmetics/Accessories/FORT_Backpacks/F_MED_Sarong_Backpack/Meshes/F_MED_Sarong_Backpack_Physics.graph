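# Physics asset for the F_MED_Sarong_Backpack cosmetic back accessory, defining bodies and joints for its petal bones.
# One large box covers the pack bone, and small boxes cover each dyn_petal bone in center, left and right groups.
# Every petal constrains directly to the pack in a star layout rather than chains, sharing tight swing and twist limits that allow only slight independent wobble.
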PhysicsAsset /BRCosmetics/Accessories/FORT_Backpacks/F_MED_Sarong_Backpack/Meshes/F_MED_Sarong_Backpack_Physics
bodies {
 dyn_petal_01 [{shapeType=Box, extentX=3.4896, extentY=3.1279, extentZ=1.5585, center=<1.3556 0 0.0003>}]
 dyn_petal_02 [{shapeType=Box, extentX=3.4896, extentY=3.1279, extentZ=1.5585, center=<1.14 0 0>}]
 dyn_petal_03 [{shapeType=Box, extentX=3.4896, extentY=3.1279, extentZ=1.5585, center=<1.0522 0 0.0001>}]
 dyn_petal_04 [{shapeType=Box, extentX=3.4896, extentY=3.1279, extentZ=1.5585, center=<1.3189 0 -0.0002>}]
 dyn_petal_05 [{shapeType=Box, extentX=3.4896, extentY=3.1279, extentZ=1.5585, center=<1.099 0 0>}]
 dyn_petal_l_01 [{shapeType=Box, extentX=3.0202, extentY=2.225, extentZ=1.3167, center=<1.3128 0 -0.0001>}]
 dyn_petal_l_02 [{shapeType=Box, extentX=3.0202, extentY=2.225, extentZ=1.3167, center=<0.886 0 0.0001>}]
 dyn_petal_l_03 [{shapeType=Box, extentX=3.0202, extentY=2.225, extentZ=1.3167, center=<1.2202 0 0>}]
 dyn_petal_l_04 [{shapeType=Box, extentX=3.0202, extentY=2.225, extentZ=1.3167, center=<1.4107 0 0>}]
 dyn_petal_l_05 [{shapeType=Box, extentX=3.0202, extentY=2.225, extentZ=1.3167, center=<1.3514 0 0>}]
 dyn_petal_l_06 [{shapeType=Box, extentX=3.0202, extentY=2.225, extentZ=1.3167, center=<1.2422 0 0>}]
 dyn_petal_l_07 [{shapeType=Box, extentX=3.0202, extentY=2.225, extentZ=1.3167, center=<1.0165 0.2836 0>}]
 dyn_petal_l_08 [{shapeType=Box, extentX=3.0202, extentY=2.225, extentZ=1.3167, center=<1.1393 0 0>}]
 dyn_petal_l_09 [{shapeType=Box, extentX=3.0202, extentY=2.225, extentZ=1.3167, center=<1.2341 0 0>}]
 dyn_petal_r_01 [{shapeType=Box, extentX=3.0202, extentY=2.225, extentZ=1.3167, center=<0.7781 0 0>}]
 dyn_petal_r_010 [{shapeType=Box, extentX=3.0202, extentY=2.225, extentZ=1.3167, center=<1.2149 0 0>}]
 dyn_petal_r_02 [{shapeType=Box, extentX=3.0202, extentY=2.225, extentZ=1.3167}]
 dyn_petal_r_03 [{shapeType=Box, extentX=3.0202, extentY=2.225, extentZ=1.3167, center=<1.028 0 0>}]
 dyn_petal_r_04 [{shapeType=Box, extentX=3.0202, extentY=2.225, extentZ=1.3167, center=<0.9837 0 0>}]
 dyn_petal_r_05 [{shapeType=Box, extentX=3.0202, extentY=2.225, extentZ=1.3167, center=<0.688 0 0>}]
 dyn_petal_r_06 [{shapeType=Box, extentX=3.0202, extentY=2.225, extentZ=1.3167, center=<1.1266 0 0>}]
 dyn_petal_r_07 [{shapeType=Box, extentX=3.0202, extentY=2.225, extentZ=1.3167, center=<0.9433 0 0.0001>}]
 dyn_petal_r_08 [{shapeType=Box, extentX=3.0202, extentY=2.225, extentZ=1.3167, center=<0.8341 0 -0.0001>}]
 dyn_petal_r_09 [{shapeType=Box, extentX=3.0202, extentY=2.225, extentZ=1.3167, center=<1.1169 0 0>}]
 pack [{shapeType=Box, extentX=24.0308, extentY=6.74, extentZ=22.7788, center=<0 -1.4124 -5.0579>}]
}
constraints [
 {bone1Name=dyn_petal_01, bone2Name=pack, swing1LimitDegrees=10, swing2LimitDegrees=15, twistLimitDegrees=5}
 {bone1Name=dyn_petal_02, bone2Name=pack, swing1LimitDegrees=10, swing2LimitDegrees=15, twistLimitDegrees=5}
 {bone1Name=dyn_petal_03, bone2Name=pack, swing1LimitDegrees=10, swing2LimitDegrees=15, twistLimitDegrees=5}
 {bone1Name=dyn_petal_04, bone2Name=pack, swing1LimitDegrees=10, swing2LimitDegrees=15, twistLimitDegrees=5}
 {bone1Name=dyn_petal_05, bone2Name=pack, swing1LimitDegrees=10, swing2LimitDegrees=15, twistLimitDegrees=5}
 {bone1Name=dyn_petal_r_01, bone2Name=pack, swing1LimitDegrees=10, swing2LimitDegrees=15, twistLimitDegrees=5}
 {bone1Name=dyn_petal_r_02, bone2Name=pack, swing1LimitDegrees=10, swing2LimitDegrees=15, twistLimitDegrees=5}
 {bone1Name=dyn_petal_r_03, bone2Name=pack, swing1LimitDegrees=10, swing2LimitDegrees=15, twistLimitDegrees=5}
 {bone1Name=dyn_petal_r_04, bone2Name=pack, swing1LimitDegrees=10, swing2LimitDegrees=15, twistLimitDegrees=5}
 {bone1Name=dyn_petal_r_05, bone2Name=pack, swing1LimitDegrees=10, swing2LimitDegrees=15, twistLimitDegrees=5}
 {bone1Name=dyn_petal_r_06, bone2Name=pack, swing1LimitDegrees=10, swing2LimitDegrees=15, twistLimitDegrees=5}
 {bone1Name=dyn_petal_r_07, bone2Name=pack, swing1LimitDegrees=10, swing2LimitDegrees=15, twistLimitDegrees=5}
 {bone1Name=dyn_petal_r_08, bone2Name=pack, swing1LimitDegrees=10, swing2LimitDegrees=15, twistLimitDegrees=5}
 {bone1Name=dyn_petal_r_09, bone2Name=pack, swing1LimitDegrees=10, swing2LimitDegrees=15, twistLimitDegrees=5}
 {bone1Name=dyn_petal_r_010, bone2Name=pack, swing1LimitDegrees=10, swing2LimitDegrees=15, twistLimitDegrees=5}
 {bone1Name=dyn_petal_l_01, bone2Name=pack, swing1LimitDegrees=10, swing2LimitDegrees=15, twistLimitDegrees=5}
 {bone1Name=dyn_petal_l_02, bone2Name=pack, swing1LimitDegrees=10, swing2LimitDegrees=15, twistLimitDegrees=5}
 {bone1Name=dyn_petal_l_03, bone2Name=pack, swing1LimitDegrees=10, swing2LimitDegrees=15, twistLimitDegrees=5}
 {bone1Name=dyn_petal_l_04, bone2Name=pack, swing1LimitDegrees=10, swing2LimitDegrees=15, twistLimitDegrees=5}
 {bone1Name=dyn_petal_l_05, bone2Name=pack, swing1LimitDegrees=10, swing2LimitDegrees=15, twistLimitDegrees=5}
 {bone1Name=dyn_petal_l_06, bone2Name=pack, swing1LimitDegrees=10, swing2LimitDegrees=15, twistLimitDegrees=5}
 {bone1Name=dyn_petal_l_07, bone2Name=pack, swing1LimitDegrees=10, swing2LimitDegrees=15, twistLimitDegrees=5}
 {bone1Name=dyn_petal_l_08, bone2Name=pack, swing1LimitDegrees=10, swing2LimitDegrees=15, twistLimitDegrees=5}
 {bone1Name=dyn_petal_l_09, bone2Name=pack, swing1LimitDegrees=10, swing2LimitDegrees=15, twistLimitDegrees=5}
]
