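# Animation Blueprint for the F_MED_Sarong back bling that tilts and physically simulates the backpack.
# Each update reads the owning mesh component's spine_05 delta from reference pose relative to spine_01, remapping its yaw into a clamped roll.
# That rotation and a small offset are applied additively to the pack bone over the reference pose.
# A RigidBody node then simulates the result using its own physics asset, overridden Y-axis gravity and damped component acceleration.
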AnimBlueprint /BRCosmetics/Accessories/FORT_Backpacks/F_MED_Sarong_Backpack/Meshes/F_MED_Sarong_Backpack_AnimBlueprint
target_skeleton=&/F_MED_Sarong_Backpack_Skeleton
graphs [
 AnimGraph {nodes [
   n1 = "Transform (Modify) Bone - Bone: pack"(ComponentPose=Local To Component(LocalPose=Local Space Ref Pose()), Translation="0,-1,0", Rotation=Get backpack_rotate()) {
    rotationMode=BMM_Additive
    rotationSpace=BCS_BoneSpace
    translationMode=BMM_Additive
    alphaBoolBlend {blendOption=Linear}
    boneToModify {boneName=pack}
    translation {y=-1}
   }
   n2 = RigidBody(ComponentPose=@n1) {
    bOverrideWorldGravity=true
    overridePhysicsAsset=&/F_MED_Sarong_Backpack_Physics
    alphaBoolBlend {blendOption=Linear}
    componentAppliedLinearAccClamp=<1000 1000 1000>
    componentLinearAccScale=<0.15 0.15 0.15>
    overrideWorldGravity {y=200}
   }
   Output Pose(Result=Component To Local(ComponentPose=@n2)) {layerGroup=None}
  ]}
 EventGraph {nodes [
   Event BlueprintUpdateAnimation()
   n1 = GetOwnerSkeletalMeshComponent()
   n2 = BreakRotator(InRot=BreakTransform(InTransform=GetDeltaTransformFromRefPose(self=@n1, BoneName=spine_05, BaseName=spine_01)).Rotation)
   > Set backpack_rotate(backpack_rotate=MakeRotator(Roll=MapRangeClamped(Value=@n2.Yaw, InRangeA=0, InRangeB=-30, OutRangeA=5, OutRangeB=40)))
   TryGetPawnOwner()
  ]}
]
variables [backpack_rotate {type=Rotator}]
